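; Looping gameplay cue notify for the charging phase of the Cherry Fireball item ability, built to signal when the charge reaches maximum.
; On loop start it calls the parent logic, then sets a one-shot timer using the cue's raw magnitude as the time to max charge.
; The timer fires ReachedMaxCharge, a hook for changing looping cosmetics that currently only prints a debug string.
; The stored timer handle is cleared and invalidated when the cue is destroyed.
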(event Custom|ReachedMaxCharge
  ;[1 1 1] When the ability hits max charge for changing any looping cosmetic logic.
  (Development|PrintString 0 "GC Hit Max Charge"))

(event EventOnLoopingStart(Cascade)
  (params (MyTarget :type Actor%20Object%20Reference) (Parameters :type Gameplay%20Cue%20Parameters%20Structure%20%28by%20ref%29) (ParticleComponents :type Array%20of%20Cascade%20Particle%20System%20Component%20Object%20References) (AudioComponents :type Array%20of%20Audio%20Component%20Object%20References))
  (bind (_normalizedmagnitude _rawmagnitude) (Ability|GameplayCue|BreakGameplayCueParameters Parameters))
  ;[1 1 1] When the ability hits max charge for changing any looping cosmetic logic.
  (|CallParentFunction MyTarget Parameters ParticleComponents AudioComponents)
  (bind _returnvalue (Utilities|Time|SetTimerbyEvent (delegate Custom|ReachedMaxCharge) _rawmagnitude))
  (Variables|Default|SetTimer_ReachedMaxCharge _returnvalue))

(event EventDestroyed
  ;[1 1 1] Clear Timer on destruction
  (Utilities|Time|ClearandInvalidateTimerbyHandle 0 (Variables|Default|GetTimer_ReachedMaxCharge)))

(fn ConstructionScript)
